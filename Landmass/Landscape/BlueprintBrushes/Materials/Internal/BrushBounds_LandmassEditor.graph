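# Translucent, unlit Landmass editor material, drawn without depth test, that outlines a landscape brush's bounds and shape.
# TilingAAGrid on each UV axis draws border lines, with widths compensated for object scale; a pixel-depth and field-of-view term supplies camera-distance scaling.
# A SphereMask around 0.5 in the ShapeRT red channel traces the shape contour.
# Bounds and contour are tinted by SelectionColor and ShapeColor, then added, with adjustable shape opacity.
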Material /Landmass/Landscape/BlueprintBrushes/Materials/Internal/BrushBounds_LandmassEditor
graph {
 properties {bDisableDepthTest=true, blendMode=BLEND_Translucent, pixelDepthOffsetMode=PDOM_Legacy, shadingModel=MSM_Unlit, usedShadingModels=Unlit}
 nodes [
  n1 = TextureCoordinate()
  n2 = Divide(
   A=ScalarParameter() {defaultValue=0.1, parameterName=LineWidth}
   B=Divide(A=MaterialFunctionCall() {materialFunction=&ObjectScale}.Scale X) {constB=256}
  )
  n3 = NamedRerouteUsage()
  n4 = TextureCoordinate()
  n5 = Divide(
   A=ScalarParameter() {defaultValue=0.1, parameterName=LineWidth}
   B=Divide(A=MaterialFunctionCall() {materialFunction=&ObjectScale}.Scale Y) {constB=256}
  )
  n6 = NamedRerouteUsage()
  n7 = Max(
   A=MaterialFunctionCall!(ScalarUV=ComponentMask(None=@n1) {r=true}, Split=Multiply(A=@n2, B=@n3)) {materialFunction=/Engine/EditorMaterials/TilingAAGrid}.Result
   B=MaterialFunctionCall(ScalarUV=ComponentMask(None=@n4) {g=true}, Split=Multiply(A=@n5, B=@n6)) {materialFunction=&TilingAAGrid}.Result
  )
  n8 = TextureSampleParameter2D() {
   parameterName=ShapeRT
   samplerType=SAMPLERTYPE_LinearColor
   texture=/Engine/Functions/Engine_MaterialFunctions02/PivotPainter2/Black_1x1_EXR_Texture
  }
  n9 = ScalarParameter() {defaultValue=0.1, parameterName=LineWidth}
  n10 = ScalarParameter() {defaultValue=2048, parameterName=ShapeMaskLineAdjust}
  n11 = SphereMask(
   A=@n8.R
   B=Constant() {r=0.5}
   Radius=Divide(
    A=Multiply(A=@n9, B=@n10)
    B=MaterialFunctionCall() {materialFunction=&ObjectScale}.Scale X
   )
  )
  Add(
   A=@n7
   B=Multiply(
    A=NamedRerouteDeclaration(None=@n11) {name=Shape Mask, nodeColor=<rgba 1 0 0.7969 1>}
    B=ScalarParameter() {defaultValue=1, parameterName=ShapeOpacity}
   )
  )
  Add(
   A=Multiply(A=VectorParameter() {parameterName=ShapeColor, defaultValue=<rgba 0 1 0.5 0>}.RGB, B=NamedRerouteUsage())
   B=Multiply(A=@n7, B=VectorParameter() {parameterName=SelectionColor, defaultValue=<rgba 0 1 1 0>}.RGB)
  )
  n12 = ViewProperty() {property=MEVP_TanHalfFieldOfView}
  n13 = ComponentMask(None=ViewProperty() {property=MEVP_ViewSize}.Property) {g=true, r=true}
  NamedRerouteDeclaration(None=Multiply(A=PixelDepth().R, B=Divide(A=@n12.Property, B=@n13))) {name=CameraDist, nodeColor=<rgba 0 1 0.7188 1>}
  Saturate(None=Divide(A=Subtract(A=@n8.R) {constB=0.5}) {constB=0.005})
 ]
}
