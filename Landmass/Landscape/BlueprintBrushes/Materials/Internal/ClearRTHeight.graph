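# ClearRTHeight is an internal unlit material for Landmass landscape blueprint brushes, which by its name clears a height render target.
# It feeds a zero constant into the PackRG8Height material function as the signed height, producing the packed encoding of zero height.
# Negative emissive color is allowed.
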Material /Landmass/Landscape/BlueprintBrushes/Materials/Internal/ClearRTHeight
graph {
 properties {bAllowNegativeEmissiveColor=true, pixelDepthOffsetMode=PDOM_Legacy, shadingModel=MSM_Unlit, usedShadingModels=Unlit}
 nodes [MaterialFunctionCall!(Height Int -32k to 32k=Constant()) {materialFunction=/Landmass/Landscape/BlueprintBrushes/MF/PackRG8Height}]
}
