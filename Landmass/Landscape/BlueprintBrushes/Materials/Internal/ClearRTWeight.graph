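# ClearRTWeight is an internal unlit material for Landmass Blueprint brushes, named for clearing a weight render target.
# Its graph holds a single Constant node left at its default value, with no parameters or texture inputs.
# Negative emissive color is allowed, though the file shows no link from the constant to a material output.
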Material /Landmass/Landscape/BlueprintBrushes/Materials/Internal/ClearRTWeight
graph {
 properties {bAllowNegativeEmissiveColor=true, pixelDepthOffsetMode=PDOM_Legacy, shadingModel=MSM_Unlit, usedShadingModels=Unlit}
 nodes [Constant()]
}
