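# Unlit internal utility material for Landmass landscape brushes that, as its name indicates, copies a render target using landscape UVs.
# It samples the RT texture parameter (black by default) at texel-center UVs from the PointSampleLandscapeUVs function, pinned to mip 0.
# Full float precision and allowed negative emissive color let it carry signed, high-precision data without clamping negatives.
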Material /Landmass/Landscape/BlueprintBrushes/Materials/Internal/CopyRT_LandscapeUVs
graph {
 properties {
  bAllowNegativeEmissiveColor=true
  floatPrecisionMode=MFPM_Full
  pixelDepthOffsetMode=PDOM_Legacy
  shadingModel=MSM_Unlit
  usedShadingModels=Unlit
 }
 nodes [TextureSampleParameter2D!(
   UVs=MaterialFunctionCall() {materialFunction=/Landmass/Landscape/BlueprintBrushes/MF/PointSampleLandscapeUVs}.TexelCenterUVs
   Level=Constant()
   Apply View MipBias
  ) {mipValueMode=TMVM_MipLevel, parameterName=RT, texture=&Black}]
}
