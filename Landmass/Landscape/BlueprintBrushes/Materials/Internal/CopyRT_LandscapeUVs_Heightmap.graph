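# Unlit Landmass brush material that copies a heightmap render target while offsetting its height relative to a brush Z position.
# Samples the RT parameter at landscape texel centers on a fixed mip, then unpacks the RG8 height.
# Subtracts the BrushZ offset from landscape Z, scaled by landscape Z scale from MPC_Landscape, then repacks to RG8.
# Full float precision and allowed negative emissive values are enabled.
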Material /Landmass/Landscape/BlueprintBrushes/Materials/Internal/CopyRT_LandscapeUVs_Heightmap
graph {
 properties {
  bAllowNegativeEmissiveColor=true
  floatPrecisionMode=MFPM_Full
  pixelDepthOffsetMode=PDOM_Legacy
  shadingModel=MSM_Unlit
  usedShadingModels=Unlit
 }
 nodes [
  n1 = TextureSampleParameter2D!(
   UVs=MaterialFunctionCall() {materialFunction=/Landmass/Landscape/BlueprintBrushes/MF/PointSampleLandscapeUVs}.TexelCenterUVs
   Level=Constant()
   Apply View MipBias
  ) {mipValueMode=TMVM_MipLevel, parameterName=RT, texture=&Black}
  n2 = Subtract(
   A=ScalarParameter() {parameterName=BrushZ}
   B=CollectionParameter() {collection=/Landmass/Landscape/BlueprintBrushes/MPC/MPC_Landscape, parameterName=LandscapeZLocation}
  )
  n3 = CollectionParameter() {collection=&MPC_Landscape, parameterName=LandscapeZScale}
  MaterialFunctionCall!(Height Int -32k to 32k=Subtract(
    A=MaterialFunctionCall!(RG=@n1.RGB) {materialFunction=/Landmass/Landscape/BlueprintBrushes/MF/UnpackRG8Height}.Result
    B=Divide(A=@n2, B=@n3)
   )) {materialFunction=/Landmass/Landscape/BlueprintBrushes/MF/PackRG8Height}
 ]
}
